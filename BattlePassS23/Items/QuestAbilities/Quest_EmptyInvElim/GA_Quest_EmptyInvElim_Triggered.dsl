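; Event-triggered Battle Pass S23 quest ability for EmptyInvElim, apparently crediting eliminations made while carrying no weapons.
; Activation scans the owning player's inventory, debug-printing item names and flagging definitions matching Tag Container.
; Without a match or Excluded Tags on the pawn, it sends the Athena.Quests.EmptyInvElim stat to the quest manager.
; The stat send and EndAbility run only when the item array is non-empty; the empty branch does nothing.
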(event Ability|EventActivateAbilityFromEvent
  (params (EventData :type Gameplay%20Event%20Data%20Structure%20%28by%20ref%29))
  (|CallParentFunction EventData)
  (Variables|Default|SetHas%20Weapons false)
  (bind _items (InventoryManagement|GetItemsfromInventoryOwner (Fortnite|GetFortPlayerControllerfromActor (Variables|GAVars|Data|GetOwningPlayerPawn))))
  (if (Utilities|Array|IsNotEmpty _items)
    (for _array_element _items
      (Development|PrintString (Utilities|GetDisplayName (Item|GetItemDefinition _array_element)) true true "(R=0.000000,G=0.660000,B=1.000000,A=1.000000)" 2e+01)
      (if (GameplayTags|HasAnyMatchingGameplayTags (Item|GetItemDefinition _array_element) (Variables|Default|GetTag%20Container))
        (Variables|Default|SetHas%20Weapons true)))
    (if (not (Variables|Default|GetHas%20Weapons))
      (if (not (GameplayTags|HasAnyMatchingGameplayTags (Variables|GAVars|Data|GetOwningPlayerPawn) (Variables|Default|GetExcluded%20Tags)))
        (Quest|SendComplexCustomStatEvent (Variables|GAVars|Data|GetOwningPlayerQuestManager) 0 "(GameplayTags=((TagName=\"Athena.Quests.EmptyInvElim\")))" "(GameplayTags=)" 1)))
    (Ability|EndAbility)
    (else
      _)))
